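# MF_DoggusWet layers an animated wetness effect onto Doggus or Cattus character material attributes, driven by MPC_Doggus.
# Pre-skinned positions drive panned front/side projections, distorted by noise and base normals and blended by local vertex normal.
# The resulting masks alter base color (with distance blending), metallic, specular, roughness and normals, plus emissive sparkles.
# The IsDoggus switch picks the wetness value, AO-masked for Cattus; ES2/ES3.1 only adjusts roughness.
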MaterialFunction /Game/Characters/Player/Male/Medium/Bodies/M_MED_Doggus/MaterialSet/MF_DoggusWet
graph {nodes [
  n1 = Transform(None=VertexNormalWS()) {transformSourceType=TRANSFORMSOURCE_World, transformType=TRANSFORM_Local}
  n2 = ComponentMask(None=@n1) {b=true}
  Abs(None=@n2)
  Constant3Vector() {constant=<rgba 20 18 25 1>}
  Constant3Vector() {constant=<rgba 0 1 0 1>}
  Constant() {r=3}
  Constant() {r=0.1}
  Constant() {r=-0.03}
  Constant() {r=50}
  Constant() {r=-1}
  Constant() {r=-0.5}
  Constant() {r=50}
  Constant() {r=5}
  Constant() {r=0.25}
  n3 = MaterialFunctionCall!(
   Blend Range=ScalarParameter() {defaultValue=70000, group=Wet, parameterName=Wet_DiffuseMod-BlendRange}
   Start Offset=ScalarParameter() {defaultValue=-15000, group=Wet, parameterName=Wet_DiffuseMod-BlendRangeOffset}
  ) {materialFunction=/Engine/Functions/Engine_MaterialFunctions03/Particles/Distance_Blend}
  n4 = ScalarParameter() {defaultValue=3, group=Wet, parameterName=Wet_MasterUV-MaskScale}
  n5 = VectorParameter() {group=Wet, parameterName=MasterUV-Scale, defaultValue=<rgba 20 20 25 1>}
  n6 = Divide(A=VertexInterpolator(VS=PreSkinnedPosition()).PS, B=Multiply(A=Abs(None=@n5.RGB)) {constB=-1})
  n7 = Time()
  n8 = MaterialFunctionCall(
   X=ScalarParameter() {group=Wet, parameterName=Wet_MasterUV-PanSpeedX}
   Y=ScalarParameter() {defaultValue=-0.2, group=Wet, parameterName=Wet_MasterUV-PanSpeedY}
  ) {materialFunction=&MakeFloat2}
  n9 = TextureSample(UVs=Multiply(A=TextureCoordinate(), B=ScalarParameter() {defaultValue=15, group=Wet, parameterName=Wet_DistortScale})) {samplerSource=SSM_Wrap_WorldGroupSettings, samplerType=SAMPLERTYPE_Normal, texture=/Game/Global/Textures/Default/Noise/T_SwirlNoise_N}
  n10 = ScalarParameter() {defaultValue=-2, group=Wet, parameterName=Wet_DistortAmount}
  n11 = ScalarParameter() {defaultValue=10, group=Wet, parameterName=Wet_DistortByBaseNormalAmount}
  n12 = GetMaterialAttributes(None=FunctionInput() {inputType=FunctionInput_MaterialAttributes}) {attributeGetTypes [
    #1
    #2
    #3
    #4
    #5
    #6
    #7
   ]}
  n13 = Transform(None=Add(A=Multiply(A=@n9.RGB, B=@n10), B=Multiply(A=@n11, B=@n12.Normal))) {transformType=TRANSFORM_Local}
  n14 = Add(A=Panner(Coordinate=ComponentMask(None=@n6) {b=true, r=true}, Time=@n7, Speed=@n8.Result), B=ComponentMask(None=@n13) {b=true, r=true})
  n15 = Add(A=Panner(Coordinate=ComponentMask(None=@n6) {b=true, g=true}, Time=@n7, Speed=@n8.Result), B=ComponentMask(None=@n13) {b=true, g=true})
  n16 = Saturate(None=Abs(None=ComponentMask(None=@n1) {r=true}))
  n17 = ScalarParameter() {defaultValue=200, group=Wet, parameterName=Wet_MasterUV-MaskMod}
  n18 = Saturate(None=Add(A=Multiply(A=@n16, B=@n17), B=ScalarParameter() {defaultValue=-0.5, group=Wet, parameterName=Wet_MasterUV-MaskAdd}))
  n19 = LinearInterpolate(
   A=TextureSample(UVs=Multiply(A=@n4, B=@n14)) {samplerSource=SSM_Wrap_WorldGroupSettings, texture=/Game/Effects/Fort_Effects/Textures/Tile/TankMoteNoise}.R
   B=TextureSample(UVs=Multiply(A=@n4, B=@n15)) {samplerSource=SSM_Wrap_WorldGroupSettings, texture=&TankMoteNoise}.R
   Alpha=@n18
  )
  n20 = ScalarParameter() {defaultValue=3, group=Wet, parameterName=Wet_NormalMaskMod}
  n21 = ScalarParameter() {defaultValue=-0.5, group=Wet, parameterName=Wet_NormalMaskAddNear}
  n22 = ScalarParameter() {defaultValue=1, group=Wet, parameterName=Wet_NormalMaskAddFar}
  n23 = Saturate(None=Add(A=Multiply(A=@n19, B=@n20), B=LinearInterpolate(A=@n21, B=@n22, Alpha=@n3.Result)))
  n24 = Add(A=@n23, B=ScalarParameter() {defaultValue=1, group=Wet, parameterName=Wet_DiffuseModMaskFarAdd})
  n25 = Multiply(
   A=LinearInterpolate(
    A=ScalarParameter() {defaultValue=0.25, group=Wet, parameterName=Wet_DiffuseModNear2}
    B=ScalarParameter() {defaultValue=0.75, group=Wet, parameterName=Wet_DiffuseModFar2}
    Alpha=@n3.Result
   )
   B=Saturate(None=LinearInterpolate(A=@n23, B=@n24, Alpha=@n3.Result))
  )
  n26 = TextureSample(UVs=@n14) {
   desc=Front
   samplerSource=SSM_Wrap_WorldGroupSettings
   samplerType=SAMPLERTYPE_LinearGrayscale
   texture=/Game/Characters/Player/Male/Medium/Bodies/M_MED_Doggus/Textures/T_DoggusWet_FX
  }
  n27 = TextureSample(UVs=@n15) {
   desc=Sides
   samplerSource=SSM_Wrap_WorldGroupSettings
   samplerType=SAMPLERTYPE_LinearGrayscale
   texture=&T_DoggusWet_FX
  }
  n28 = Power(
   Base=LinearInterpolate(A=@n26.R, B=@n27.R, Alpha=@n18)
   Exp=ScalarParameter() {defaultValue=4, group=Wet, parameterName=Wet_MasterMask-Power}
  )
  n29 = ScalarParameter() {defaultValue=5, group=Wet, parameterName=Wet_MasterMask-Mod}
  n30 = Saturate(None=Add(A=Multiply(A=@n28, B=@n29), B=ScalarParameter() {defaultValue=-0.2, group=Wet, parameterName=Wet_MasterMask-Add}))
  n31 = LinearInterpolate(
   A=@n12.BaseColor
   B=Desaturation(None=Add(A=@n25, B=@n12.BaseColor), Fraction=ScalarParameter() {defaultValue=-1, group=Wet, parameterName=Wet_DiffuseModSaturation})
   Alpha=@n30
  )
  n32 = FunctionInput() {bUsePreviewValueAsDefault=true, inputName=AO_Mask_Input, inputType=FunctionInput_Scalar, previewValue=<1 1 1 1>}
  n33 = Add(
   A=Multiply(A=@n32, B=ScalarParameter() {defaultValue=1, group=Wet, parameterName=Wet_AO-MaskMod})
   B=ScalarParameter() {defaultValue=1, group=Wet, parameterName=Wet_AO-MaskAdd}
  )
  n34 = StaticSwitchParameter(
   True=CollectionParameter() {collection=&/MPC_Doggus, parameterName=Wet_Doggus}
   False=Multiply(
    A=Saturate(None=@n33)
    B=CollectionParameter() {collection=&MPC_Doggus, parameterName=Wet_Cattus}
   )
  ) {defaultValue=true, parameterName=IsDoggus}
  n35 = ScalarParameter() {defaultValue=0.5, group=Wet, parameterName=Wet_Metal}
  n36 = ScalarParameter() {defaultValue=1, group=Wet, parameterName=Wet_Spec}
  n37 = ScalarParameter() {group=Wet, parameterName=Wet_Roughness}
  n38 = TextureSample(UVs=@n14) {
   samplerSource=SSM_Wrap_WorldGroupSettings
   samplerType=SAMPLERTYPE_Normal
   texture=/Game/Characters/Player/Male/Medium/Bodies/M_MED_Doggus/Textures/T_DoggusWet_N
  }
  n39 = TextureSample(UVs=@n15) {
   samplerSource=SSM_Wrap_WorldGroupSettings
   samplerType=SAMPLERTYPE_Normal
   texture=&T_DoggusWet_N
  }
  n40 = LinearInterpolate(A=@n38.RGB, B=@n39.RGB, Alpha=@n18)
  n41 = MaterialFunctionCall(Normal=@n40, Flatness=OneMinus(None=@n30)) {materialFunction=&FlattenNormal}
  n42 = TextureCoordinate()
  n43 = ScalarParameter() {defaultValue=8, group=Wet, parameterName=Wet_SparkleScale}
  n44 = ScalarParameter() {defaultValue=1, group=Wet, parameterName=Wet_MasterUV-MaskScale2}
  n45 = Multiply(A=@n44, B=@n14)
  n46 = Multiply(A=@n44, B=@n15)
  n47 = Multiply(
   A=LinearInterpolate(
    A=TextureSample(UVs=@n45) {samplerSource=SSM_Wrap_WorldGroupSettings, texture=/Game/Effects/Fort_Effects/Textures/Tile/T_Fractal_01}.R
    B=TextureSample(UVs=@n46) {samplerSource=SSM_Wrap_WorldGroupSettings, texture=&T_Fractal_01}.R
    Alpha=@n18
   )
   B=TextureSample(UVs=Multiply(A=@n42, B=@n43)) {samplerSource=SSM_Wrap_WorldGroupSettings, texture=&T_Fractal_01}.R
  )
  n48 = ScalarParameter() {defaultValue=20, group=Wet, parameterName=Wet_SparkleMod}
  n49 = Saturate(None=Add(A=Multiply(A=@n47, B=@n48), B=ScalarParameter() {defaultValue=-0.5, group=Wet, parameterName=Wet_SparkleAdd}))
  n50 = Multiply(
   A=VectorParameter() {group=Wet, parameterName=Wet_EmissiveColor, defaultValue=<rgba 0.9688 1.065 1.5 1>}.RGB
   B=Multiply(A=@n30, B=Multiply(A=@n23, B=@n49))
  )
  n51 = SetMaterialAttributes!(
   MaterialAttributes=@n12.MaterialAttributes
   BaseColor=LinearInterpolate(A=@n12.BaseColor, B=Saturate(None=@n31), Alpha=@n34)
   Metallic=LinearInterpolate(A=@n12.Metallic, B=LinearInterpolate(A=@n12.Metallic, B=@n35, Alpha=@n30), Alpha=@n34)
   Specular=LinearInterpolate(A=@n12.Specular, B=LinearInterpolate(A=@n12.Specular, B=@n36, Alpha=@n30), Alpha=@n34)
   Roughness=LinearInterpolate(A=@n12.Roughness, B=LinearInterpolate(A=@n12.Roughness, B=@n37, Alpha=@n30), Alpha=@n34)
   Normal=LinearInterpolate(
    A=@n12.Normal
    B=MaterialFunctionCall(BaseNormal=@n12.Normal, AdditionalNormal=@n41.Result) {materialFunction=&BlendAngleCorrectedNormals}.Result
    Alpha=@n34
   )
   EmissiveColor=LinearInterpolate(A=@n12.EmissiveColor, B=Add(A=@n50, B=@n12.EmissiveColor), Alpha=@n34)
  ) {attributeSetTypes [
    #1
    #2
    #3
    #4
    #5
    #6
   ]}
  n52 = SetMaterialAttributes!(
   MaterialAttributes=@n12.MaterialAttributes
   Roughness=LinearInterpolate(A=@n12.Roughness, B=ScalarParameter() {defaultValue=0.2, group=Wet, parameterName=Wet_Mobile}, Alpha=@n34)
  ) {attributeSetTypes [#4]}
  FunctionOutput(None=FeatureLevelSwitch(Default=@n51, ES2=@n52, ES3_1=@n52, SM4_REMOVED=@n51))
  n53 = MaterialFunctionCall(Normal=@n40, Flatness=@n23) {materialFunction=&FlattenNormal}
  n54 = TextureCoordinate()
  n55 = ScalarParameter() {defaultValue=40, group=Wet, parameterName=Wet_BumpNormal-Scale}
  n56 = Time()
  n57 = MaterialFunctionCall(X=ScalarParameter() {group=Wet, parameterName=Wet_BumpNormal-Speed}, Y=ScalarParameter() {group=Wet, parameterName=Wet_BumpNormal-Speed2}) {materialFunction=&MakeFloat2}
  n58 = TextureSample(UVs=Add(A=Multiply(A=@n54, B=@n55), B=Multiply(A=@n56, B=@n57.Result))) {samplerSource=SSM_Wrap_WorldGroupSettings, samplerType=SAMPLERTYPE_Normal, texture=&T_SwirlNoise_N}
  n59 = MaterialFunctionCall(
   Normal=Multiply(A=@n58.RGB, B=VectorParameter() {group=Wet, parameterName=Wet-BumpNormal-Amount, defaultValue=<rgba 1 1 1 1>}.RGB)
   Flatness=OneMinus(None=@n23)
  ) {materialFunction=&FlattenNormal}
  Reroute(None=Normalize(VectorInput=Add(A=@n53.Result, B=@n59.Result)))
  n60 = Saturate(None=@n2)
  n61 = ScalarParameter() {defaultValue=300, group=Wet, parameterName=bbbbg}
  Saturate(None=Add(A=Multiply(A=@n60, B=@n61), B=ScalarParameter() {defaultValue=-2, group=Wet, parameterName=bbbbg4}))
  TextureSample() {samplerSource=SSM_Wrap_WorldGroupSettings, texture=&TankMoteNoise}
  Transform()
 ]}
